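# Parent editor-utility report for the Generic Asset Review Tool that validates lights on actors or single light components.
# For actors without an editor-only root, embedded Python gathers owned, non-TRASH light components and runs RunComponentLogic on each.
# The base component check fails visible, world-affecting, non-editor-only lights with positive intensity, but supplies no error text.
# Actor success depends only on the last component's error array, which each loop iteration overwrites.
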EditorUtilityBlueprint /Game/NeverCook/EditorUtilities/GenericAssetReviewTool/BP/ReportExamples/Actor/Report_ActorLight_Parent
graphs [
 AddComponentErrors {nodes [
   n1 = AddComponentErrors!() -> (then, Array:Array of Strings, Object:Object Reference) {is=title K2Node_FunctionEntry}
   > Branch(Condition=Is Valid Index!(TargetArray=@n1, IndexToTest:int) -> (ReturnValue:bool))
   > AddError(NewError=Append(A=GetCompName!(Object:Object Reference=@n1.Object) -> (ReturnValue:String), B=" has errors."))
   > n2 = For Each Loop(Array=@n1)
   > AddError(NewError=@n2, Depth=1)
  ]}
 EventGraph {links [], nodes []}
 GetCompName {nodes [
   n1 = GetCompName!() -> (then, Object:Object Reference) {is=title K2Node_FunctionEntry}
   n2 = GetDisplayName(Object=@n1)
   n3 = Split(SourceString=@n2, InStr=., SearchDir=FromStart)
   > n4 = Branch(Condition=@n3.ReturnValue)
   > Return Node!(execute, ReturnValue:String=@n3.RightS)
   Return Node(execute=@n4.else, ReturnValue=@n2)
  ]}
 RunActorLogic {nodes [
   n1 = RunActorLogic!() -> (then, Actor:Actor Object Reference) {is=title K2Node_FunctionEntry}
   > Set l_actor(l_actor:Actor Object Reference=@n1) -> (Output_Get:Actor Object Reference)
   > n2 = Sequence!(execute) -> (then_0, then_1, then_2:Exec)
   n3 = Get l_actor() -> (l_actor:Actor Object Reference)
   > Branch(Condition=Get bIsEditorOnly!(self:Actor Component Object Reference=Get RootComponent!(self:Actor Object Reference=@n3) -> (RootComponent:Scene Component Object Reference)) -> (bIsEditorOnly:bool))
   > Return Node!(execute, Success?:bool=true, severity:Enum_GART_ReportSeverity Enum=NewEnumerator0)
   n4 = Execute Python Script!(
    execute=@n2.then_1
    PythonScript="import unreal\r\n\r\nit = unreal.ObjectIterator(unreal.LightComponent)\r\nreturn_array= []\r\nfor x in it:\r\n    actor = x.get_outer()\r\n    if actor == in_actor:\r\n        if \"TRASH\" not in str(x):\r\n            return_array.append(x)\r\n"
    in_actor:Actor Object Reference=Get l_actor()
   ) -> (then, ReturnValue, return_array:Array of Light Component Object References) {inputs [in_actor], outputs [return_array]}
   > Clear(TargetArray=Get l_error() -> (l_error:Array of Strings))
   > n5 = Set l_comps(l_comps:Array of Light Component Object References=@n4.return_array) -> (Output_Get:Array of Light Component Object References)
   > n6 = For Each Loop(Array:Array of Light Component Object References=@n5) -> (Array Element:Light Component Object Reference)
   > n7 = Set l_comp(l_comp:Light Component Object Reference=@n6) -> (Output_Get:Light Component Object Reference)
   > n8 = RunComponentLogic!(execute, self, Comp:Light Component Object Reference=@n7) -> (then, Success?:bool, Error:Array of Strings, severity:Enum_GART_ReportSeverity Enum)
   > Set l_error(l_error:Array of Strings=@n8.Error) -> (Output_Get:Array of Strings)
   > n9 = Branch(Condition=@n8)
   AddComponentErrors!(
    execute=@n9.else
    self
    Array:Array of Strings=Get l_error()
    Object:Object Reference=Get l_comp() -> (l_comp:Light Component Object Reference)
   ) -> (then) {is=title K2Node_CallFunction}
   > IncreaseErrorSeverity!(
    execute
    self:GART Report Base Object Reference
    Severity:"Enum_GART_ReportSeverity Enum (by ref)"=Get l_severity()
    NewSeverity:Enum_GART_ReportSeverity Enum=@n8.severity
   ) -> (then)
   n10 = Get l_error()
   Return Node(execute=@n2.then_2, Success?="integer < integer"(A=Length(TargetArray:Array of Strings=@n10), B=1), severity=Get l_severity())
   SetSeverity!(execute, self:GART Report Base Object Reference, NewSeverity:Enum_GART_ReportSeverity Enum=NewEnumerator0, OnlyIfHigher:bool=true) -> (then)
  ]}
 RunComponentLogic {nodes [
   n1 = RunComponentLogic!() -> (then, Comp:Light Component Object Reference) {is=title K2Node_FunctionEntry}
   > Set l_comp(l_comp=@n1)
   n2 = Get l_comp()
   n3 = Get bIsEditorOnly(self=Get l_comp())
   n4 = Get Intensity!(self:Light Component Base Object Reference=@n2) -> (Intensity:float)
   > n5 = Branch(Condition=AND Boolean!(
     A=IsVisible(self:Scene Component Object Reference=@n2)
     B=Get bAffectsWorld!(self:Light Component Base Object Reference=@n2) -> (bAffectsWorld:bool)
     C:bool="float > float"(A=@n4)
     D:bool=NOT Boolean(A=@n3)
    ) -> (ReturnValue))
   > Return Node!(execute, Success?=false, Error:Array of Strings, severity=NewEnumerator0)
   Return Node(execute=@n5.else, Success?=true, severity=NewEnumerator0)
  ]}
 RunFilterLogicOnObject {nodes [
   n1 = RunFilterLogicOnObject!() -> (then, object:Object Reference) {is=title K2Node_FunctionEntry}
   > n2 = Sequence!(execute) -> (then_0, then_1)
   > n3 = Cast To Actor!(execute, Object=@n1) -> (then, CastFailed, AsActor:Actor Object Reference) {is=title K2Node_DynamicCast}
   > n4 = RunActorLogic!(execute, self, Actor:Actor Object Reference=@n3) -> (then, Success?:bool, severity:Enum_GART_ReportSeverity Enum) {is=title K2Node_CallFunction}
   > Set l_success(l_success:bool=@n4) -> (Output_Get:bool)
   > Set l_severity(l_severity:Enum_GART_ReportSeverity Enum=@n4.severity) -> (Output_Get:Enum_GART_ReportSeverity Enum)
   n5 = Cast To LightComponent!(execute=@n3.CastFailed, Object=@n1) -> (then, CastFailed, AsLight Component:Light Component Object Reference) {is=title K2Node_DynamicCast}
   > n6 = RunComponentLogic(Comp=@n5) {is=title K2Node_CallFunction}
   > n7 = Set l_success(l_success=@n6)
   > n8 = Branch(Condition=@n7)
   AddComponentErrors(execute=@n8.else, Array=@n6.Error, Object=@n5)
   > Set l_severity(l_severity=@n6.severity)
   Return Node!(
    execute=@n2.then_1
    Valid:bool=Get l_success() -> (l_success:bool)
    ReportData:Struct GART Report Data Structure=Make Struct GART Report Data(
     DisplayTitle_20_#1=Actor
     DisplayDescShort_29_#2=Get FinalDisplayDescShort() -> (FinalDisplayDescShort:String)
     Severity_46_#3=Get l_severity()
    )
   )
  ]}
 setFinalDescShort {nodes [
   n1 = setFinalDescShort!() -> (then, FinalDisplayDescShort:String) {is=title K2Node_FunctionEntry}
   > Set FinalDisplayDescShort(FinalDisplayDescShort:String=@n1) -> (Output_Get:String)
  ]}
]
variables [
 FinalDisplayDescShort {type=string, default=Lights}
 LightDatabase {type=BP_LightReviewTool_Database_C, default=/Game/NeverCook/ArtTools/Tech_Art/LightReviewTool/DA_LightReviewToolData}
]
